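# Niagara deploy effect for the BirdNest glider, built from six purple glow and lightning emitters, all disabled in this file.
# AirGlow, Corona and Flare_Glow supply sprite, mesh and light flashes, with front-end/gameplay modifiers adjusting the latter two.
# LightningMesh001 and LightningMesh_Burst ramp mesh spawn rates with particle lights, the former tinting its lights from System.Overload.
# LargeArcs001 builds curl-noise ribbons partitioned per burst, with platform-dependent point counts.
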NiagaraSystem /BRCosmetics/Gadgets/Assets/VinderTech_GliderChute/Glider_BirdNest/FX/NS_Glider_BirdNest_Deploy
dependencies {
 usedDataInterfaces [&NiagaraDataInterfaceCurve, &NiagaraDataInterfaceColorCurve, &NiagaraDataInterfacePlatformSet]
 usedDynamicInputs []
 usedModules [&SystemState]
}
emitters {
 AirGlow {
  bEnabled=false
  emitterUpdateScript [
   &EmitterState(Life Cycle Mode=::NewEnumerator1, Loop Behavior=::NewEnumerator1, Loop Delay=0.5, Loop Duration=1, UseLoopDelay=false)
   &SpawnBurst_Instantaneous(Spawn Count=1, Use Spawn Probability=false)
  ]
  particleSpawnScript [/Niagara/Modules/Spawn/Initialization/InitializeParticle(
    Color=<rgba 0.8317 0 2 0.2>
    Module.Write Color=true
    Module.Write Lifetime=true
    Module.Write Mass=false
    Module.Write Material Random=true
    Module.Write Position=false
    Module.Write Scale=false
    Module.Write SpriteRotation=false
    Module.Write SpriteSize=true
    Position=&SimulationPosition()
    Sprite Size=&Vector2DFromFloat(Value=800)
   )]
  particleUpdateScript [
   &ParticleState()
   &ScaleColor(
    Scale Alpha=&FloatFromCurve(FloatCurve=DataInterface() {propertyValues {Curve {keys [
         <at 0 0 RCIM_Cubic>
         {arriveTangent=0.4483, interpMode=RCIM_Cubic, leaveTangent=0.4483, tangentMode=RCTM_User, time=0.1, value=1}
         <at 1 0 RCIM_Cubic>
        ]}}})
    ScaleA=true
    ScaleRGB=true
   )
   &ScaleSpriteSize(Scale Factor=&Vector2DFromFloat(Value=&FloatFromCurve(FloatCurve=DataInterface() {propertyValues {Curve {keys [<at 0 0 RCIM_Cubic>,<at 0.2 1 RCIM_Cubic>]}}})))
   &DynamicMaterialParameters(
    Index 0 Param 1=2
    Module.Param0WriteEnabled=true
    Module.Param1WriteEnabled=true
    Module.Param2WriteEnabled=true
    Module.Param3WriteEnabled=true
   )
   &CameraOffset(Camera Offset Amount=-200)
  ]
  rendererClasses [&NiagaraSpriteRendererProperties]
  renderers [{rendererClass=&NiagaraSpriteRendererProperties}]
 }
 Corona {
  bEnabled=false
  emitterSpawnScript [/SetVariables(Emitter.SizingCurve=DataInterface() {propertyValues {Curve {keys [
        {arriveTangent=155.1764, interpMode=RCIM_Cubic, leaveTangent=155.1764, tangentMode=RCTM_User, time=0, value=0}
        {arriveTangentWeight=0.175, interpMode=RCIM_Cubic, leaveTangentWeight=0.1588, tangentMode=RCTM_User, time=0.0262, value=1.8822}
        <at 0.15 1>
        <at 0.75 1>
        <at 1 0>
       ]}}}) {isSetParameters=true, module=SetVariables_#1}]
  emitterUpdateScript [
   &EmitterState(
    Inactive Response=::NewEnumerator1
    Life Cycle Mode=::NewEnumerator1
    Loop Behavior=::NewEnumerator1
    Loop Delay=&Add_Float(A=$System.GlobalDelay, B=-0.05)
    Loop Duration=4
    UseLoopDelay=true
   )
   &SpawnBurst_Instantaneous(Spawn Time=0.225, Use Spawn Probability=false, UseLoopCountLimit=false)
  ]
  particleSpawnScript [
   /Niagara/Modules/Spawn/Initialization/V2/InitializeParticle(
    Color=<rgba 0.4159 0 1 1>
    Mesh Scale Mode=::NewEnumerator3
    Mesh Uniform Scale=4
    Sprite Rotation Mode=::NewEnumerator2
    Sprite Size Mode=::NewEnumerator3
    Uniform Sprite Size=&FN_FrontEndFloatModifier(Gameplay=750, Front End=200)
    UsePositionOffset=true
    Write Material Random=true
   )
   &CameraOffset(Camera Offset Amount=&FN_FrontEndFloatModifier(Gameplay=75, Front End=20))
  ]
  particleUpdateScript [
   &ParticleState()
   &ScaleSpriteSize(Uniform Curve Sprite Scale=DataInterface() {propertyValues {Curve {keys [
        {arriveTangent=4.7631, interpMode=RCIM_Cubic, leaveTangent=4.7631, tangentMode=RCTM_User, time=0, value=0.2}
        {arriveTangent=0.8, interpMode=RCIM_Cubic, leaveTangent=0.8, time=0.25, value=0.8}
        {interpMode=RCIM_Cubic, leaveTangent=0.0224, time=1, value=1}
       ]}}})
   /SetVariables(Particles.Corona Color:LinearColor=&MultiplyLinearColor(
     Linear Color=$Particles.Initial.Color
     Scale Factor=&MakeFloat4FromVectorAndFloat(
      "Vector (RGB)"=&VectorFromFloat(Value=1)
      "Float (Alpha)"=&FloatFromCurve(FloatCurve=DataInterface() {propertyValues {Curve {keys [
           {arriveTangent=8.9568, interpMode=RCIM_Cubic, leaveTangent=8.9568, tangentMode=RCTM_User, time=0, value=0}
           {arriveTangentWeight=0.1341, interpMode=RCIM_Cubic, leaveTangentWeight=0.2413, tangentMode=RCTM_User, time=0.1, value=1}
           {arriveTangentWeight=0.0833, interpMode=RCIM_Cubic, tangentMode=RCTM_User, time=1, value=0}
          ]}}})
     )
    )) {isSetParameters=true, module=SetVariables_#2}
  ]
  rendererClasses [&NiagaraSpriteRendererProperties]
  renderers [{rendererClass=&NiagaraSpriteRendererProperties}]
 }
 Flare_Glow {
  bEnabled=false
  emitterSpawnScript [/SetVariables(Emitter.SizingCurve=DataInterface() {propertyValues {Curve {keys [
        {arriveTangent=60.4226, interpMode=RCIM_Cubic, leaveTangent=60.4226, tangentMode=RCTM_User, time=0, value=0}
        {arriveTangentWeight=0.175, interpMode=RCIM_Cubic, leaveTangentWeight=0.1588, tangentMode=RCTM_User, time=0.05, value=1.8822}
        {arriveTangentWeight=0.0833, interpMode=RCIM_Cubic, leaveTangent=-7.6948, time=0.593, value=0}
       ]}}}) {isSetParameters=true, module=SetVariables_#1}]
  emitterUpdateScript [
   &EmitterState(Inactive Response=::NewEnumerator1, Life Cycle Mode=::NewEnumerator1, Loop Behavior=::NewEnumerator1, Loop Duration=4, UseLoopDelay=true)
   &SpawnBurst_Instantaneous(Use Spawn Probability=false, UseLoopCountLimit=false)
  ]
  particleSpawnScript [
   &InitializeParticle(
    Color=&LerpLinearColors(
     StartColor=<rgba 6.2381 0 15 1>
     EndColor=<rgba 2.9111 0 7 0.5>
     LerpFactor=&FN_FrontEndFloatModifier(Gameplay=0, Front End=1)
    )
    Lifetime=0.5
    Mesh Scale Mode=::NewEnumerator3
    Mesh Uniform Scale=&FN_FrontEndFloatModifier(Gameplay=2, Front End=1.2)
    Sprite Rotation Mode=::NewEnumerator2
    Sprite Size Mode=::NewEnumerator3
    Uniform Sprite Size=&FN_FrontEndFloatModifier(Gameplay=200, Front End=250)
    UsePositionOffset=true
    Write Material Random=true
   )
   &InitialMeshOrientation(Mesh Orientation Mode=::NewEnumerator7, Rotation=<0 -0.25 0>, Use Rotation Vector=true)
   &InitialMeshOrientation(
    Mesh Orientation Mode=::NewEnumerator7
    Rotation=&RandomRangeVector(Minimum=<0 0 0>, Maximum=<1 0 0>)
    Use Rotation Vector=true
   ) {module=InitialMeshOrientation001}
   &CameraOffset(Camera Offset Amount=&FN_FrontEndFloatModifier(Gameplay=200, Front End=200))
  ]
  particleUpdateScript [
   &ParticleState()
   &ScaleColor(
    Scale Alpha=&Multiply_Float(
     A=&Lerp_Float(A=0.75, B=1, Alpha=&Sine(Normalized Angle=$Emitter.Age, Period=0.5, Scale=0.5, Bias=0.5))
     B=&FloatFromCurve(FloatCurve=DataInterface() {propertyValues {Curve {keys [<at 0.95 1>,<at 1 0.1>]}}})
    )
    Scale RGB=&VectorFromFloat(Value=&Multiply_Float(
      A=&FloatFromCurve(FloatCurve=$Emitter.SizingCurve)
      B=&FNTimeOfDayScalar(Night Scale=0.5, Morning Scale=1, Evening Scale=1)
     ))
    ScaleA=true
    ScaleRGB=true
   ) {module=ScaleColor001}
   &ScaleMeshSize(Scale Factor=&VectorFromFloat(Value=&FloatFromCurve(FloatCurve=$Emitter.SizingCurve, Scale Curve=1.25)))
   &ScaleSpriteSize(Uniform Curve Sprite Scale:NiagaraDataInterfaceCurve=$Emitter.SizingCurve)
   /SetVariables(
    Particles.Corona Color:LinearColor=&MultiplyLinearColor(
     Linear Color=&LerpLinearColors(
      StartColor=<rgba 2 1.02 0.04 0.1>
      EndColor=<rgba 2 1.02 0.02 0.1>
      LerpFactor=&FN_FrontEndFloatModifier(Gameplay=0, Front End=1)
     )
     Scale Factor=&MakeFloat4FromVectorAndFloat(
      "Vector (RGB)"=&VectorFromFloat(Value=1)
      "Float (Alpha)"=&FloatFromCurve(FloatCurve=DataInterface() {propertyValues {Curve {keys [
           {arriveTangent=8.9568, interpMode=RCIM_Cubic, leaveTangent=8.9568, tangentMode=RCTM_User, time=0, value=0}
           {arriveTangentWeight=0.1341, interpMode=RCIM_Cubic, leaveTangentWeight=0.2413, tangentMode=RCTM_User, time=0.3, value=1}
           {arriveTangentWeight=0.1593, interpMode=RCIM_Cubic, leaveTangentWeight=0.0885, tangentMode=RCTM_User, time=0.75, value=1}
           {arriveTangentWeight=0.0833, interpMode=RCIM_Cubic, tangentMode=RCTM_User, time=1, value=0}
          ]}}})
     )
    )
    Particles.Corona Size:Vector2f=&Multiply_Vector2D_ByFloat(
     Float=&Lerp_Float(
      A=0.9
      B=1
      Alpha=&FN_FrontEndFloatModifier(
       Gameplay=&FloatFromCurve(FloatCurve=DataInterface() {propertyValues {Curve {keys [
            {arriveTangent=60.4226, interpMode=RCIM_Cubic, leaveTangent=60.4226, tangentMode=RCTM_User, time=0, value=0}
            {arriveTangentWeight=0.175, interpMode=RCIM_Cubic, leaveTangentWeight=0.1588, tangentMode=RCTM_User, time=0.05, value=1}
            {arriveTangentWeight=0.1341, interpMode=RCIM_Cubic, leaveTangentWeight=0.2413, tangentMode=RCTM_User, time=0.3, value=0.5313}
            {arriveTangentWeight=0.1593, interpMode=RCIM_Cubic, leaveTangentWeight=0.0885, tangentMode=RCTM_User, time=0.75, value=0.5313}
            {arriveTangentWeight=0.0833, interpMode=RCIM_Cubic, tangentMode=RCTM_User, time=1, value=0.3985}
           ]}}})
       Front End=&FloatFromCurve(FloatCurve=DataInterface() {propertyValues {Curve {keys [
            {arriveTangent=60.4226, interpMode=RCIM_Cubic, leaveTangent=60.4226, tangentMode=RCTM_User, time=0, value=0}
            {arriveTangentWeight=0.175, interpMode=RCIM_Cubic, leaveTangentWeight=0.1588, tangentMode=RCTM_User, time=0.05, value=1}
            {arriveTangentWeight=0.1341, interpMode=RCIM_Cubic, leaveTangentWeight=0.2413, tangentMode=RCTM_User, time=0.3, value=0.5313}
            {arriveTangentWeight=0.1593, interpMode=RCIM_Cubic, leaveTangentWeight=0.0885, tangentMode=RCTM_User, time=0.75, value=0.5313}
            {arriveTangentWeight=0.0833, interpMode=RCIM_Cubic, tangentMode=RCTM_User, time=1, value=0.3985}
           ]}}})
      )
     )
     Vector2D=<800 800>
    )
   ) {isSetParameters=true, module=SetVariables_#2}
   &FN_Light_Attributes(
    Color=$Particles.Color
    Create Color Attribute=true
    Create Exponent Attribute=true
    Create Radius Attribute=true
    Exponent=2
    Radius=300
    Scale Color=4
   )
  ]
  rendererClasses [
   &NiagaraMeshRendererProperties
   &NiagaraSpriteRendererProperties
   &NiagaraLightRendererProperties
  ]
  renderers [
   {rendererClass=&NiagaraMeshRendererProperties, rendererIndex=0}
   {rendererClass=&NiagaraSpriteRendererProperties, rendererIndex=1}
   {rendererClass=&NiagaraLightRendererProperties, rendererIndex=2}
  ]
 }
 LargeArcs001 {
  bEnabled=false
  emitterSpawnScript [/SetVariables(Emitter.num ribbon particles:NiagaraInt32=&MakeCustomIntFromBool(
     Bool=&IsPlatformSetActive(New Platform Set=DataInterface() {propertyValues {Platforms {qualityLevelMask=-1}}})
     True INT=5
     False INT=10
    )) {isSetParameters=true}]
  emitterUpdateScript [
   &EmitterState(
    Life Cycle Mode=::NewEnumerator1
    Loop Behavior=::NewEnumerator2
    Loop Count:NiagaraInt32=&RandomRangeInt(Minimum=8, Maximum=10)
    Loop Duration=0.075
    Recalculate Duration Each Loop=true
    UseLoopDelay=false
   )
   &SpawnBurst_Instantaneous(Spawn Count=$Emitter.num ribbon particles, Use Spawn Probability=false, UseLoopCountLimit=false)
   /SetVariables(
    Emitter.Radom Height:NiagaraFloat=&RandomRangeFloat(Minimum=50, Maximum=75)
    Emitter.random Float:NiagaraFloat=&RandomRangeFloat(Minimum=0, Maximum=1)
   ) {isSetParameters=true}
   /SetVariables(
    Emitter.ShapeX:NiagaraFloat=&RandomRangeFloat(Minimum=0, Maximum=1, Evaluation Type=::NewEnumerator1)
    Emitter.ShapeY:NiagaraFloat=&RandomRangeFloat(Minimum=0, Maximum=1, Evaluation Type=::NewEnumerator1)
   ) {isSetParameters=true}
  ]
  particleSpawnScript [
   &InitializeRibbon(
    Color=&RandomRangeLinearColor(Minimum=<rgba 0.8317 0 2 1>, Maximum=<rgba 0.4159 0 1 1>)
    Lifetime=&UniformRangedFloat(Minimum=0.2, Maximum=0.35)
    Ribbon Width=20
    UsePositionOffset=true
    Write Material Random=true
   )
   &PartitionParticles(
    Particles Per Partition:NiagaraInt32=$Emitter.num ribbon particles
    Partition Count Mode:ENiagara_InfiniteLoopDuration=::NewEnumerator1
    Partition Mode:ENiagara_PartitionMode=::NewEnumerator1
    Write Partition Index to Ribbon Link Order:ENiagara_AttributeSamplingApplyOutput=::NewEnumerator0
    Write Partition to Ribbon ID=::NewEnumerator0
   )
   /SetVariables(Particles.PartitionParticles.Normalized Index In Partition:NiagaraFloat=&Multiply_Float(
     A=$Particles.PartitionParticles.Normalized Index In Partition
     B=&Divide_Float(
      A=1
      B=&OneMinusFloat(Float=&Divide_Float(A=1, B=&MakeFloatFromInt(Int=$Emitter.num ribbon particles)))
     )
    )) {isSetParameters=true}
   &ShapeLocation(
    MaskSpawnGroup=false
    Sphere Distribution=::NewEnumerator1
    Sphere Radius Position:NiagaraFloat=$Particles.PartitionParticles.Normalized Index In Partition
    U Position=$Emitter.ShapeX
    UseOwnerScale=true
    V Position=$Emitter.ShapeY
   )
  ]
  particleUpdateScript [
   &ParticleState()
   &ScaleColor(
    Scale Alpha=&FloatFromCurve(FloatCurve=DataInterface() {propertyValues {Curve {keys [
         {arriveTangent=8.0553, interpMode=RCIM_Cubic, leaveTangent=8.0553, tangentMode=RCTM_User, time=0, value=0}
         {arriveTangentWeight=0.1171, interpMode=RCIM_Cubic, leaveTangentWeight=0.3, tangentMode=RCTM_User, time=0.2113, value=1}
         {arriveTangentWeight=0.2913, interpMode=RCIM_Cubic, leaveTangentWeight=0.1333, tangentMode=RCTM_User, time=0.7524, value=1}
         {arriveTangent=-10.4882, interpMode=RCIM_Cubic, leaveTangent=-10.4882, tangentMode=RCTM_User, time=1, value=0}
        ]}}})
    Scale RGB=&VectorFromFloat(Value=0.35)
    ScaleA=true
    ScaleRGB=true
   )
   &ScaleRibbonWidth(Ribbon Width Scale=&Multiply_Float(
     A=&FloatFromCurve(FloatCurve=DataInterface() {propertyValues {Curve {keys [
          {interpMode=RCIM_Cubic, tangentMode=RCTM_User, time=0, value=0}
          {interpMode=RCIM_Cubic, tangentMode=RCTM_User, time=0.25, value=1}
          {arriveTangentWeight=0.1491, interpMode=RCIM_Cubic, leaveTangentWeight=0.2236, tangentMode=RCTM_User, time=0.75, value=1}
          {arriveTangentWeight=0.1, interpMode=RCIM_Cubic, tangentMode=RCTM_User, time=1, value=0}
         ]}}})
     B=&FloatFromCurve(
      FloatCurve=DataInterface() {propertyValues {Curve {keys [
          {interpMode=RCIM_Cubic, leaveTangentWeight=0.0667, tangentMode=RCTM_User, time=0, value=0}
          {arriveTangentWeight=0.1491, interpMode=RCIM_Cubic, leaveTangentWeight=0.2236, tangentMode=RCTM_User, time=0.1128, value=1}
          {arriveTangentWeight=0.1021, interpMode=RCIM_Cubic, leaveTangentWeight=0.1021, tangentMode=RCTM_User, time=0.5, value=1}
          {arriveTangentWeight=0.2236, interpMode=RCIM_Cubic, leaveTangentWeight=0.1491, tangentMode=RCTM_User, time=0.9515, value=1}
          {arriveTangentWeight=0.0667, interpMode=RCIM_Cubic, tangentMode=RCTM_User, time=1, value=0}
         ]}}}
      CurveIndex=$Particles.PartitionParticles.Normalized Index In Partition
     )
    )) {module=ScaleRibbonWidth001}
   &CurlNoiseForce(Noise Quality / Cost=::NewEnumerator0, Noise Strength=250, Pan Noise=false)
   &SolveForcesAndVelocity(Clamp Velocity=false, Limit Acceleration=false)
   &SphereLocation(
    MaskSpawnGroup=false
    Override Seed=false
    Sphere Origin:NiagaraPosition=$Particles.Position
    Sphere Radius=&Multiply_Float(
     A=$Particles.PartitionParticles.Normalized Index In Partition
     B=&RandomRangeFloat(Minimum=0, Maximum=5)
    )
    Surface Only=false
    UseNonUniformScale=false
    UseOffset=false
    UseSphereOrientation=false
   )
   /SetVariables(Particles.Scale Pos:Vector3f=&Lerp_Vector(
     A=/Niagara/DynamicInputs/Helpers/SimulationPosition()
     B=Linked() {linkedVariable {name=Particles.Position, type=NiagaraPosition}}
     Alpha=&Multiply_Float(A=3, B=$Particles.NormalizedAge)
    )) {isSetParameters=true}
   &Color(Color=&RandomRangeLinearColor(Minimum=<rgba 2 0.5 0 1>, Maximum=<rgba 1 0.961 0 1>)) {enabled=false}
  ]
  rendererClasses [&NiagaraRibbonRendererProperties]
  renderers [{rendererClass=&NiagaraRibbonRendererProperties}]
 }
 LightningMesh001 {
  bEnabled=false
  emitterUpdateScript [
   &EmitterState(
    Life Cycle Mode=::NewEnumerator1
    Loop Behavior=::NewEnumerator1
    Loop Delay=&RandomRangeFloat(Minimum=0, Maximum=0)
    Loop Duration=1
    UseLoopDelay=false
   )
   &SpawnRate(
    SpawnRate=&FloatFromCurve(FloatCurve=DataInterface() {propertyValues {Curve {keys [<at 0 1>,<at 0.5 20>]}}}, CurveIndex=$Emitter.Age)
    Use Spawn Probability=true
   )
   /SetVariables(Emitter.RandomPosition:NiagaraFloat=&RandomRangeFloat(Minimum=0, Maximum=1, Randomness Mode=::NewEnumerator1, Random Seed=$Emitter.LoopCount, Fixed Random Seed=true)) {isSetParameters=true}
  ]
  particleSpawnScript [
   &InitializeParticle(
    Color=<rgba 0.4159 0 1 1>
    Lifetime=&FloatFromCurve(FloatCurve=DataInterface() {propertyValues {Curve {keys [<at 0 0.25 RCIM_Cubic>, {arriveTangent=0.4797, interpMode=RCIM_Cubic, leaveTangent=0.4797, tangentMode=RCTM_User, time=1, value=0.4}]}}})
    Mesh Scale=&Multiply_VectorByFloat(
     Vector=&MaskVectorBySpawnGroup(Masked Vector=<1 1 1>, Passthrough Vector=<1 1 -1>)
     Float=&RandomRangeFloat(Minimum=4, Maximum=8)
    )
    Mesh Scale Mode=::NewEnumerator4
    Sprite Size Mode=::NewEnumerator3
    Uniform Sprite Size=100
    UsePositionOffset=false
    Write Material Random=true
   )
   &InitialMeshOrientation(Use Rotation Vector=true)
   /SetVariables(
    Particles.LightningSubImageIndex:NiagaraFloat=&UniformRangedFloat(Minimum=0, Maximum=3)
    Particles.SpriteSize=&Multiply_Vector2D_ByFloat(
     Float=&MaskFloatBySpawnGroup(Masked Float=&RandomRangeFloat(Minimum=1, Maximum=2), Passthrough Float=1, Spawn Group Mask=1)
     Vector2D=$Particles.SpriteSize
    )
   ) {isSetParameters=true}
  ]
  particleUpdateScript [
   &ParticleState()
   &ShapeLocation(MaskSpawnGroup=false, Override Seed=false, Rotation Mode=::NewEnumerator2, UseOwnerScale=true, Yaw / Pitch / Roll=<0 0.25 0>)
   &ScaleMeshSize(Scale Factor=&VectorFromFloat(Value=1.5))
   &ScaleColor(
    Scale Alpha=&FloatFromCurve(FloatCurve=DataInterface() {propertyValues {Curve {keys [<at 0.5 1 RCIM_Cubic>,<at 1 0 RCIM_Cubic>]}}})
    Scale RGB=&VectorFromFloat(Value=&FloatFromCurve(FloatCurve=DataInterface() {propertyValues {Curve {keys [
          <at 0 0 RCIM_Cubic>
          {arriveTangent=15, interpMode=RCIM_Cubic, leaveTangent=15, time=0.1, value=20}
          {arriveTangent=-22.2222, interpMode=RCIM_Cubic, leaveTangent=-22.2222, time=0.2, value=3}
          <at 1 0 RCIM_Cubic>
         ]}}}))
    ScaleA=true
    ScaleRGB=true
   )
   &CameraOffset(Camera Offset Amount=10)
   &DynamicMaterialParameters(
    Index 0 Param 1=&FloatFromCurve(FloatCurve=DataInterface() {propertyValues {Curve {keys [<at 0 1>,<at 1 0>]}}})
    Index 0 Param 2=&FloatFromCurve(
     FloatCurve=DataInterface() {propertyValues {Curve {keys [
         <at 0 1 RCIM_Cubic>
         {arriveTangent=0.987, interpMode=RCIM_Cubic, leaveTangent=0.987, tangentMode=RCTM_User, time=0.4992, value=1}
         <at 0.571 1 RCIM_Cubic>
         <at 1 1>
        ]}}}
     Scale Curve=5000
    )
    Index 0 Param 3=&RandomRangeFloat(Minimum=1, Maximum=2)
    Index 0 Param 4=&RandomRangeFloat(Minimum=0, Maximum=1)
    Index 1 Param 1=1
    Index 1 Param 3=1
    Module.Param0WriteEnabled=true
    Module.Param0WriteEnabled1=true
    Module.Param1WriteEnabled=true
    Module.Param1WriteEnabled1=true
    Module.Param2WriteEnabled=true
    Module.Param2WriteEnabled1=true
    Module.Param3WriteEnabled=true
    Module.Param3WriteEnabled1=true
    Write Parameter Index 1=true
   ) {module=DynamicMaterialParameters004}
   /SetVariables(Particles.LightColor=&LerpLinearColors(
     StartColor=<rgba 1 2.3042 5 1>
     EndColor=<rgba 5 1.6667 1 1>
     LerpFactor=&FloatFromCurve(FloatCurve=DataInterface() {propertyValues {Curve {keys [<at 0.8 0>,<at 1 1>]}}}, CurveIndex=$System.Overload)
    )) {isSetParameters=true}
   &FN_Light_Attributes(
    Color=$Particles.LightColor
    Create Color Attribute=true
    Create Exponent Attribute=true
    Create Position Attribute=true
    Create Radius Attribute=true
    Exponent=2
    Position=&Add_Vector(
     A=Linked() {linkedVariable {name=Particles.Position, type=NiagaraPosition}}
     B=&Multiply_VectorByFloat(Vector=$Particles.ShapeLocation.ShapeNormal, Float=100)
    )
    Radius=200
    Scale Color=&FloatFromCurve(
     FloatCurve=DataInterface() {propertyValues {Curve {keys [
         <at 0 0 RCIM_Cubic>
         {arriveTangent=0.875, interpMode=RCIM_Cubic, leaveTangent=0.875, time=0.1, value=0.9}
         {arriveTangent=-0.2131, interpMode=RCIM_Cubic, leaveTangent=-0.2131, tangentMode=RCTM_User, time=0.4, value=0.35}
         {arriveTangent=-0.5833, interpMode=RCIM_Cubic, leaveTangent=-0.5833, time=0.7, value=0.5}
         <at 1 0 RCIM_Cubic>
        ]}}}
     Scale Curve=1.5
    )
   )
   &ScaleColor(
    Scale RGB=&Lerp_Vector(
     A=&MakeVectorFromLinearColorRGB(Color=<rgba 0 0.3261 1 1>)
     B=&MakeVectorFromLinearColorRGB(Color=<rgba 1 0.2267 0.2 1>)
     Alpha=&FloatFromCurve(FloatCurve=DataInterface() {propertyValues {Curve {keys [<at 0.8 0>,<at 1 1>]}}}, CurveIndex=$System.Overload)
    )
    ScaleA=true
    ScaleRGB=true
   ) {enabled=false, module=ScaleColor001}
  ]
  rendererClasses [
   &NiagaraMeshRendererProperties
   &NiagaraSpriteRendererProperties
   &NiagaraLightRendererProperties
  ]
  renderers [
   {rendererClass=&NiagaraMeshRendererProperties, rendererIndex=0}
   {rendererClass=&NiagaraSpriteRendererProperties, rendererIndex=1}
   {rendererClass=&NiagaraLightRendererProperties, rendererIndex=2}
   {rendererClass=&NiagaraSpriteRendererProperties, rendererIndex=3}
  ]
 }
 LightningMesh_Burst {
  bEnabled=false
  emitterUpdateScript [
   &EmitterState(
    Inactive Response=::NewEnumerator1
    Life Cycle Mode=::NewEnumerator1
    Loop Behavior=::NewEnumerator1
    Loop Delay=2
    Loop Duration=1
    UseLoopDelay=false
   )
   &SpawnRate(
    SpawnRate=&FloatFromCurve(FloatCurve=DataInterface() {propertyValues {Curve {keys [<at 0 1>,<at 0.5 20>]}}}, CurveIndex=$Emitter.Age)
    Use Spawn Probability=false
   )
   &SpawnBurst_Instantaneous(Spawn Count=10, Use Spawn Probability=false, UseLoopCountLimit=false) {enabled=false}
  ]
  particleSpawnScript [
   /Niagara/Modules/Spawn/Initialization/InitializeParticle(
    Color=<rgba 4.1587 0 10 1>
    Lifetime=&UniformRangedFloat(Minimum=0.2, Maximum=0.25)
    Mesh Scale=&UniformRangedVector(Minimum=<3 2 2>, Maximum=<3 3 3>)
    Module.Write Color=true
    Module.Write Lifetime=true
    Module.Write Mass=true
    Module.Write Material Random=true
    Module.Write Position=false
    Module.Write Scale=true
    Module.Write SpriteRotation=false
    Module.Write SpriteSize=false
    Position=/Niagara/Functions/Localspace/SimulationPosition()
   )
   &DynamicMaterialParameters(
    Index 0 Param 1=6
    Index 0 Param 3=&UniformRangedFloat(Minimum=1.45, Maximum=2.2)
    Index 0 Param 4=&UniformRangedFloat(Minimum=-60, Maximum=60)
    Module.Param0WriteEnabled=true
    Module.Param1WriteEnabled=false
    Module.Param2WriteEnabled=true
    Module.Param3WriteEnabled=true
   )
   &ShapeLocation(MaskSpawnGroup=false, Override Seed=false, Sphere Radius=50, UseOwnerScale=true)
   &InitialMeshOrientation(Use Rotation Vector=false)
  ]
  particleUpdateScript [
   &ParticleState()
   &DynamicMaterialParameters(
    Index 0 Param 2=&FloatFromCurve(FloatCurve=DataInterface() {propertyValues {Curve {keys [{arriveTangent=-2.9283, interpMode=RCIM_Cubic, leaveTangent=-2.9283, tangentMode=RCTM_User, time=0.1, value=1}, <at 1 -1>]}}})
    Module.Param0WriteEnabled=false
    Module.Param1WriteEnabled=true
    Module.Param2WriteEnabled=false
    Module.Param3WriteEnabled=false
   ) {module=DynamicMaterialParameters001}
   &ScaleColor(
    Scale RGB=&MakeVectorFromLinearColorRGB(Color=&ColorFromCurve(ColorCurve=DataInterface() {propertyValues {
        AlphaCurve {keys []}
        BlueCurve {keys [<at 0 2 RCIM_Cubic>,<at 1 4 RCIM_Cubic>]}
        GreenCurve {keys [<at 0 1.8653 RCIM_Cubic>,<at 1 2.6922 RCIM_Cubic>]}
        RedCurve {keys [<at 0 1.5312 RCIM_Cubic>,<at 1 0.8958 RCIM_Cubic>]}
       }}))
    ScaleA=true
    ScaleRGB=true
   )
   &CameraOffset(Camera Offset Amount=10)
   &FN_Light_v3(
    Color=<rgba 4.1587 0 10 1>
    Create Color Attribute=true
    Create Exponent Attribute=true
    Create Radius Attribute=true
    Enable Particle Light Enable=true
    Radius=200
   )
   /SetVariables(Particles.spritescale:Vector2f=<20 20>) {isSetParameters=true}
  ]
  rendererClasses [&NiagaraMeshRendererProperties, &NiagaraLightRendererProperties]
  renderers [
   {rendererClass=&NiagaraMeshRendererProperties, rendererIndex=0}
   {rendererClass=&NiagaraLightRendererProperties, rendererIndex=1}
  ]
 }
}
summary {
 systemName=NS_Glider_BirdNest_Deploy
 emitters [
  {
   bEnabled=false
   emitterName=LightningMesh_Burst
   rendererClasses [&NiagaraMeshRendererProperties, &NiagaraLightRendererProperties]
  }
  {bEnabled=false, emitterName=AirGlow, rendererClasses [&NiagaraSpriteRendererProperties]}
  {
   bEnabled=false
   emitterName=Flare_Glow
   rendererClasses [
    &NiagaraMeshRendererProperties
    &NiagaraSpriteRendererProperties
    &NiagaraLightRendererProperties
   ]
  }
  {bEnabled=false, emitterName=Corona, rendererClasses [&NiagaraSpriteRendererProperties]}
  {
   bEnabled=false
   emitterName=LightningMesh001
   rendererClasses [
    &NiagaraMeshRendererProperties
    &NiagaraSpriteRendererProperties
    &NiagaraLightRendererProperties
   ]
  }
  {bEnabled=false, emitterName=LargeArcs001, rendererClasses [&NiagaraRibbonRendererProperties]}
 ]
 userVariables [
  User.bIsFrontEnd {type=NiagaraBool}
  User.bIsFrontEndPreview {type=NiagaraBool}
  User.SkeletalMesh {
   type=NiagaraDataInterfaceSkeletalMesh
   defaultValue {
    struct=NiagaraExt_VariableValue_DataInterface
    value {dataInterface=&self:NiagaraDataInterfaceSkeletalMesh_0}
   }
  }
 ]
}
